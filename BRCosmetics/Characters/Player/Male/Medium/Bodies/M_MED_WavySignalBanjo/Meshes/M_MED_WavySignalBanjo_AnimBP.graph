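# Post-process style animation Blueprint for the WavySignalBanjo male body cosmetic, layering procedural motion onto the parent mesh's copied pose.
# A Control Rig driven by character state feeds belt-space RigidBody physics.
# In the front end, an idle curve blends DefaultSlot onto dagger and belt bones.
# Each update, weapon change and customization load toggles a material hide parameter when the tagged pickaxe is equipped.
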AnimBlueprint /BRCosmetics/Characters/Player/Male/Medium/Bodies/M_MED_WavySignalBanjo/Meshes/M_MED_WavySignalBanjo_AnimBP
target_skeleton=&/M_MED_WavySignalBanjo_Skeleton
graphs [
 AnimGraph {nodes [
   n1 = Control Rig!(
    Source=Copy Pose From Mesh() {bUseAttachedParent=true}
    Alpha
    skydive:bool=NOT Boolean(A=Get bIsSkyDiving())
    SkydiveFloating:bool=Get bIsSkydiveFloating()
    dbno:bool=Get bIsDBNO()
    RigidBodyEnabled:bool=Get bRigidBodyEnabled()
    crouching:bool=Get bIsCrouching()
    moving_2d:bool=Get bIsMoving2D()
   ) -> (Pose) {
    alphaBoolBlend {blendOption=Linear}
    controlRigAssetReference {blueprintRigClass=&/M_MED_WavySignalBanjo_Ode_CtrlRig.M_MED_WavySignalBanjo_Ode_CtrlRig_C}
   }
   n2 = RigidBody(ComponentPose=Local To Component(LocalPose=@n1)) {
    bOverrideWorldGravity=true
    overridePhysicsAsset=&/M_MED_WavySignalBanjo_Physics
    simulationSpace=BaseBoneSpace
    alphaBoolBlend {blendOption=Linear}
    baseBoneRef {boneName=mstr_belt_weapon}
    overrideWorldGravity {z=-980}
    simSpaceSettings {maxAngularAcceleration=30, maxAngularVelocity=10, maxLinearAcceleration=30, maxLinearVelocity=30}
   }
   n3 = Layered blend per bone(
    BasePose=Component To Local(ComponentPose=@n2)
    BlendPoses_0=Slot 'DefaultSlot'()
    BlendWeights_0=Get customIdleIsPlayingFloat() -> (customIdleIsPlayingFloat:float)
   ) {layerSetup [{branchFilters [{boneName=dagger}, {boneName=dyn_belt1}, {boneName=dyn_belt2}, {boneName=dyn_belt3}, {boneName=mstr_belt_weapon}]}]}
   Output Pose(Result=@n3) {layerGroup=None}
  ]}
 EventGraph {nodes [
   Event BlueprintUpdateAnimation()
   > n1 = Sequence()
   n2 = Get Dynamic Material Instance(execute=[@n4.then, @n5.then, @n1.then_0], SkeletalMeshComponent=GetPartSkeletalMeshComponent())
   > n3 = Branch(Condition=PickaxeItemDefinitionHasAnyMetaTag(Tags=Make Array("[0]"=(TagName="Cosmetics.MetaTags.WavySignalBanjo")), bRequiresEquipped=true))
   > SetScalarParameterValue(self=@n2, ParameterName=Hide Element 01, Value=1)
   SetScalarParameterValue(execute=@n3.else, self=@n2, ParameterName=Hide Element 01, Value=0)
   Branch(execute=@n1.then_1, Condition=Get bIsInFrontEnd())
   > Set customIdleIsPlayingFloat(customIdleIsPlayingFloat:float=GetCurveValue(self=GetOwnerFrontEndAnimInstance(), CurveName=Frontend_Idle_Bypass)) -> (Output_Get:float)
   n4 = Event OnWeaponChanged()
   n5 = Event OnCharacterCustomizationLoaded()
  ]}
]
variables [customIdleIsPlayingFloat {type=number}]
